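# GrillMoon_swb1000 is a Level Sequence that acts as the master for shot swb1000 of the GrillMoon cinematic sequences.
# It layers subsequence tracks by discipline (animation, audio, camera, gameplay, lighting, techart and VFX), all sharing one frame range at 60 fps.
# A FortStaticMeshActor mover cube has a spawn track and a constant world transform, apparently as a placement anchor for the shot.
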LevelSequence /GrillMoon/Sequence/swb/swb1000/GrillMoon_swb1000
bindings [_swb1000_MoverCube {
  id=#1
  possessed_class=FortStaticMeshActor
  tracks [
   {type=MovieSceneSpawnTrack, sections [{type=MovieSceneSpawnSection, channels [None {type=MovieSceneScriptingBoolChannel, default=true}]}]}
   {
    type=MovieScene3DTransformTrack
    sections [{
      type=MovieScene3DTransformSection
      channels [
       Location.X {type=MovieSceneScriptingDoubleChannel, default=-57361.3607}
       Location.Y {type=MovieSceneScriptingDoubleChannel, default=41946.7312}
       Location.Z {type=MovieSceneScriptingDoubleChannel, default=9847.898}
       Rotation.X {type=MovieSceneScriptingDoubleChannel, default=0}
       Rotation.Y {type=MovieSceneScriptingDoubleChannel, default=0}
       Rotation.Z {type=MovieSceneScriptingDoubleChannel, default=0}
       Scale.X {type=MovieSceneScriptingDoubleChannel, default=1}
       Scale.Y {type=MovieSceneScriptingDoubleChannel, default=1}
       Scale.Z {type=MovieSceneScriptingDoubleChannel, default=1}
      ]
     }]
   }
  ]
 }]
display_rate {denominator=1, numerator=60}
tick_resolution {denominator=1, numerator=24000}
tracks [
 {type=MovieSceneSubTrack, display_name=animation, sections [{type=MovieSceneSubSection, end_frame=4579, start_frame=3600}]}
 {type=MovieSceneSubTrack, display_name=audio, sections [{type=MovieSceneSubSection, end_frame=4579, start_frame=3600}]}
 {type=MovieSceneSubTrack, display_name=camera, sections [{type=MovieSceneSubSection, end_frame=4579, start_frame=3600}]}
 {type=MovieSceneSubTrack, display_name=gameplay, sections [{type=MovieSceneSubSection, end_frame=4579, start_frame=3600}]}
 {type=MovieSceneSubTrack, display_name=lighting, sections [{type=MovieSceneSubSection, end_frame=4579, start_frame=3600}]}
 {type=MovieSceneSubTrack, display_name=techart, sections [{type=MovieSceneSubSection, end_frame=4579, start_frame=3600}]}
 {type=MovieSceneSubTrack, display_name=vfx, sections [{type=MovieSceneSubSection, end_frame=4579, start_frame=3600}]}
]
