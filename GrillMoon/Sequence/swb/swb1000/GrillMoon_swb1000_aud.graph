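# Audio sub-sequence for shot swb1000 of the GrillMoon cinematic, holding the shot's sound separately in its own Level Sequence.
# A single master Audio track plays two overlapping sections, the shorter nested within the time span of the longer, with volume and pitch left at default.
# A GrillMoon_Countdown_MistMylar player mannequin is bound as a possessable but carries no tracks.
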LevelSequence /GrillMoon/Sequence/swb/swb1000/GrillMoon_swb1000_aud
bindings [GrillMoon_Countdown_MistMylar {id=#1, possessed_class=FortPlayerMannequin, tracks []}]
display_rate {denominator=1, numerator=60}
tick_resolution {denominator=1, numerator=24000}
tracks [{
  type=MovieSceneAudioTrack
  display_name=Audio
  sections [
   {
    type=MovieSceneAudioSection
    end_frame=4293
    start_frame=3833
    channels [
     Volume {type=MovieSceneScriptingFloatChannel, default=1}
     Pitch {type=MovieSceneScriptingFloatChannel, default=1}
     AttachActor {type=MovieSceneScriptingActorReferenceChannel}
    ]
   }
   {
    type=MovieSceneAudioSection
    end_frame=4303
    start_frame=3633
    channels [
     Volume {type=MovieSceneScriptingFloatChannel, default=1}
     Pitch {type=MovieSceneScriptingFloatChannel, default=1}
     AttachActor {type=MovieSceneScriptingActorReferenceChannel}
    ]
   }
  ]
 }]
