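# Level Sequence for the GrillMoon swb1000 event that drives camera shake and controller rumble, building to an explosion.
# A single Events_Shared_cameraForce_BP binding, with spawn and fixed-transform tracks, has its shake class and force feedback effect keyed from small shared defaults to explosion-specific assets.
# Event tracks fire Radius and Lobby variants of both effects, with Lobby triggers aligned to the asset swaps, including a second explosion shake shortly after the first.
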LevelSequence /GrillMoon/Sequence/swb/swb1000/GrillMoon_swb1000_cam
bindings [Events_Shared_cameraForce_BP {
  id=#1
  possessed_class=Events_Shared_cameraForce_BP_C
  tracks [
   {
    type=MovieSceneObjectPropertyTrack
    display_name=Force Effect
    sections [{
      type=MovieSceneObjectPropertySection
      channels [None {
        type=MovieSceneScriptingObjectPathChannel
        default="<Object '/GrillMoon/ForcedFeedback/GrillMoon_FF_swb1000_ExplosionLong' (0x8509b8b08c0) Class 'ForceFeedbackEffect'>"
        keys [
         [0, "<Object '&Events_Shared_small_FF' (0x84c3c703480) Class 'ForceFeedbackEffect'>"]
         [3844, "<Object '&GrillMoon_FF_swb1000_ExplosionLong' (0x8509b8b08c0) Class 'ForceFeedbackEffect'>"]
        ]
       }]
     }]
   }
   {
    type=MovieSceneObjectPropertyTrack
    display_name=Cam Shake Base
    sections [{
      type=MovieSceneObjectPropertySection
      channels [None {
        type=MovieSceneScriptingObjectPathChannel
        default="<Object '&Events_Shared_00small_csb_C' (0x84c3c6a8a00) Class 'BlueprintGeneratedClass'>"
        keys [
         [
          0
          "<Object '&Events_Shared_00small_csb_C' (0x84c3c6a8a00) Class 'BlueprintGeneratedClass'>"
         ]
         [
          3844
          "<Object '/GrillMoon/CameraShake/GrillMoon_kfs1000_mExpl_csb.GrillMoon_kfs1000_mExpl_csb_C' (0x850ae954200) Class 'BlueprintGeneratedClass'>"
         ]
         [
          3909
          "<Object '/GrillMoon/CameraShake/GrillMoon_kfs1000_mExpl2_csb.GrillMoon_kfs1000_mExpl2_csb_C' (0x850ae954800) Class 'BlueprintGeneratedClass'>"
         ]
        ]
       }]
     }]
   }
   {
    type=MovieSceneEventTrack
    display_name=Cam Shake Radius Event Trigger
    sections [{
      type=MovieSceneEventTriggerSection
      channels [None {type=MovieSceneScriptingEventChannel, const ["<Struct 'MovieSceneEvent' (0x8502435e100) {}>", 1]}]
     }]
   }
   {type=MovieSceneSpawnTrack, sections [{type=MovieSceneSpawnSection, channels [None {type=MovieSceneScriptingBoolChannel, default=true}]}]}
   {
    type=MovieSceneEventTrack
    display_name=Force Feedback Radius Event Trigger
    sections [{
      type=MovieSceneEventTriggerSection
      channels [None {type=MovieSceneScriptingEventChannel, const ["<Struct 'MovieSceneEvent' (0x850243544c0) {}>", 1]}]
     }]
   }
   {
    type=MovieScene3DTransformTrack
    sections [{
      type=MovieScene3DTransformSection
      channels [
       Location.X {type=MovieSceneScriptingDoubleChannel, default=3696}
       Location.Y {type=MovieSceneScriptingDoubleChannel, default=27040}
       Location.Z {type=MovieSceneScriptingDoubleChannel, default=0}
       Rotation.X {type=MovieSceneScriptingDoubleChannel, default=0}
       Rotation.Y {type=MovieSceneScriptingDoubleChannel, default=0}
       Rotation.Z {type=MovieSceneScriptingDoubleChannel, default=0}
       Scale.X {type=MovieSceneScriptingDoubleChannel, default=1}
       Scale.Y {type=MovieSceneScriptingDoubleChannel, default=1}
       Scale.Z {type=MovieSceneScriptingDoubleChannel, default=1}
      ]
     }]
   }
   {
    type=MovieSceneIntegerTrack
    sections [{type=MovieSceneIntegerSection, channels [None {type=MovieSceneScriptingIntegerChannel, default=2, const [0,1]}]}]
   }
   {
    type=MovieSceneEventTrack
    display_name=Cam Shake Lobby Event Trigger
    sections [{
      type=MovieSceneEventTriggerSection
      channels [None {
        type=MovieSceneScriptingEventChannel
        keys [[3844, "<Struct 'MovieSceneEvent' (0x85024350140) {}>"], [3909, "<Struct 'MovieSceneEvent' (0x8502435d3e0) {}>"]]
       }]
     }]
   }
   {
    type=MovieSceneEventTrack
    display_name=Force Feedback Lobby Event Trigger
    sections [{
      type=MovieSceneEventTriggerSection
      channels [None {
        type=MovieSceneScriptingEventChannel
        keys [[0, "<Struct 'MovieSceneEvent' (0x85024353d40) {}>"], [3844, "<Struct 'MovieSceneEvent' (0x85024359ba0) {}>"]]
       }]
     }]
   }
  ]
 }]
display_rate {denominator=1, numerator=60}
tick_resolution {denominator=1, numerator=24000}
tracks []
