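# GrillMoon swb1000 gameplay ("gam") Level Sequence that fires scripted events, apparently alongside the swb1000 cinematic.
# An event trigger track on a possessed BP_PlayerBuiltStructureDestroyer targets that actor, likely to clear player-built structures.
# A separate unbound event track fires two closely spaced triggers.
# It has no other tracks, and the functions the events call are not shown.
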LevelSequence /GrillMoon/Sequence/swb/swb1000/GrillMoon_swb1000_gam
bindings [BP_PlayerBuiltStructureDestroyer {
  id=#1
  possessed_class=BP_PlayerBuiltStructureDestroyer_C
  tracks [{
    type=MovieSceneEventTrack
    sections [{
      type=MovieSceneEventTriggerSection
      channels [None {type=MovieSceneScriptingEventChannel, const ["<Struct 'MovieSceneEvent' (0x8502435ed80) {}>", 1]}]
     }]
   }]
 }]
display_rate {denominator=1, numerator=60}
tick_resolution {denominator=1, numerator=24000}
tracks [{
  type=MovieSceneEventTrack
  sections [{
    type=MovieSceneEventTriggerSection
    channels [None {
      type=MovieSceneScriptingEventChannel
      keys [[3763, "<Struct 'MovieSceneEvent' (0x85024352260) {}>"], [3906, "<Struct 'MovieSceneEvent' (0x85024350c80) {}>"]]
     }]
   }]
 }]
