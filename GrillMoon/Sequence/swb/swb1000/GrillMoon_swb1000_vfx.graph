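# Level Sequence for the VFX layer of GrillMoon shot swb1000, staging a shockwave blast with a light flash and screen fade.
# A spawn track turns on the NS_GrillMoon_Shockwave Niagara actor at a fixed world position.
# Blast_Light expands its attenuation radius, shifts from white to magenta, then briefly spikes in intensity.
# A fade track reaches full as the blast builds, holds through the intensity peak, then eases back to clear.
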LevelSequence /GrillMoon/Sequence/swb/swb1000/GrillMoon_swb1000_vfx
bindings [
 NS_GrillMoon_Shockwave {
  id=#1
  possessed_class=NiagaraActor
  tracks [
   {
    type=MovieSceneSpawnTrack
    sections [{type=MovieSceneSpawnSection, channels [None {type=MovieSceneScriptingBoolChannel, default=false, keys [[3600,false],[3849,true]]}]}]
   }
   {
    type=MovieScene3DTransformTrack
    sections [{
      type=MovieScene3DTransformSection
      channels [
       Location.X {type=MovieSceneScriptingDoubleChannel, default=-57613.348}
       Location.Y {type=MovieSceneScriptingDoubleChannel, default=41967.5473}
       Location.Z {type=MovieSceneScriptingDoubleChannel, default=10450.5054}
       Rotation.X {type=MovieSceneScriptingDoubleChannel, default=0}
       Rotation.Y {type=MovieSceneScriptingDoubleChannel, default=0}
       Rotation.Z {type=MovieSceneScriptingDoubleChannel, default=0}
       Scale.X {type=MovieSceneScriptingDoubleChannel, default=1}
       Scale.Y {type=MovieSceneScriptingDoubleChannel, default=1}
       Scale.Z {type=MovieSceneScriptingDoubleChannel, default=1}
      ]
     }]
   }
  ]
 }
 Blast_Light {
  id=#2
  possessed_class=PointLight
  tracks [
   {type=MovieSceneSpawnTrack, sections [{type=MovieSceneSpawnSection, channels [None {type=MovieSceneScriptingBoolChannel, default=true}]}]}
   {
    type=MovieScene3DTransformTrack
    sections [{
      type=MovieScene3DTransformSection
      channels [
       Location.X {type=MovieSceneScriptingDoubleChannel, default=0, const [-161.6054,1]}
       Location.Y {type=MovieSceneScriptingDoubleChannel, default=0, const [134.3518,1]}
       Location.Z {type=MovieSceneScriptingDoubleChannel, default=453.0011, const [625.0749,1]}
       Rotation.X {type=MovieSceneScriptingDoubleChannel, default=0, const [0,1]}
       Rotation.Y {type=MovieSceneScriptingDoubleChannel, default=0, const [0,1]}
       Rotation.Z {type=MovieSceneScriptingDoubleChannel, default=0, const [0,1]}
       Scale.X {type=MovieSceneScriptingDoubleChannel, default=1, const [1,1]}
       Scale.Y {type=MovieSceneScriptingDoubleChannel, default=1, const [1,1]}
       Scale.Z {type=MovieSceneScriptingDoubleChannel, default=1, const [1,1]}
      ]
     }]
   }
  ]
 }
 LightComponent0 {
  id=#3
  parent_id=#2
  possessed_class=PointLightComponent
  tracks [
   {
    type=MovieSceneFloatTrack
    display_name=Intensity
    sections [{type=MovieSceneFloatSection, channels [None {type=MovieSceneScriptingFloatChannel, default=0, keys [[3882,0],[3922,2500000],[3928,0]]}]}]
   }
   {
    type=MovieSceneColorTrack
    display_name=Light Color
    sections [{
      type=MovieSceneColorSection
      channels [
       Color.R {type=MovieSceneScriptingFloatChannel, default=1, const [1,4]}
       Color.G {type=MovieSceneScriptingFloatChannel, default=1, thinned_from=4, keys [[3874,1],[3877,0],[4083,0]]}
       Color.B {type=MovieSceneScriptingFloatChannel, default=1, thinned_from=4, keys [[3874,1],[3877,0.956],[4083,0.956]]}
       Color.A {type=MovieSceneScriptingFloatChannel, default=1, const [1,4]}
      ]
     }]
   }
   {
    type=MovieSceneFloatTrack
    display_name=Attenuation Radius
    sections [{
      type=MovieSceneFloatSection
      channels [None {type=MovieSceneScriptingFloatChannel, default=7781.5952, keys [[3837,0],[3843,10],[3874,7781.5952]]}]
     }]
   }
   {type=MovieScene3DAttachTrack, sections [{type=MovieScene3DAttachSection}]}
  ]
 }
]
display_rate {denominator=1, numerator=60}
tick_resolution {denominator=1, numerator=24000}
tracks [{
  type=MovieSceneFadeTrack
  sections [{
    type=MovieSceneFadeSection
    channels [None {type=MovieSceneScriptingFloatChannel, default=1, keys [[3857,0],[3875,1],[3922,1],[3995,0.4119],[4083,0]]}]
   }]
 }]
